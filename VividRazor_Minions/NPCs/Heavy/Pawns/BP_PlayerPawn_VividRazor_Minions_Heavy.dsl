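; Pawn for the VividRazor Heavy minion NPC, which forces Walking movement and shows an AI boss-style health bar.
; The server delays enabling the bar and writes health and shield into replicated HealthInfo when health changes.
; Its OnRep updates the client widget if a local player is within range and the pawn isn't dying or bar-disabled.
; Widget visibility toggles component ticking and a looping timer that adjusts its vertical offset.
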(event EventBeginPlay
  (|CallParentFunction)
  ;[1 1 1] Force walking movement style
  (Class|FortPlayerAnimationInterface|SetMovementStyleOverride (Animation|GetAnimationInterface self) "Walking")
  (CallFunction|SetCurrentMovementStyle "Walking")
  (Networking|SwitchHasAuthority
    (:Authority
      (Health|BindEventtoOnPawnHealthChanged (delegate Custom|PawnHealthChanged))
      (Utilities|Time|SetTimerbyEvent (delegate Custom|EnableHealthBarDelayedEvent) 1.0))
    (:Remote
      (CallFunction|StartUpdatingHealthWidget))))

(event Custom|EnableHealthBarDelayedEvent
  (Variables|Default|SetEnableHealthBar true))

(event Custom|PawnHealthChanged
  (if (Variables|Default|GetShowHealth)
    (Variables|Default|SetHealthInfo (Utilities|Struct|MakeStructNPCHealthInfo (Math|Float|SafeDivide (FortPawn|GetHealth) (FortPawn|GetMaxHealth)) (Math|Float|SafeDivide (FortPawn|GetShield) (FortPawn|GetMaxShield)) (FortPawn|GetMaxHealth) (FortPawn|GetMaxShield)))))

(event EventOnMovementStyleChanged
  (params (PreviousMovementStyle :type EFortMovementStyle%20Enum :default "Running") (NewMovementStyle :type EFortMovementStyle%20Enum :default "Running"))
  ; Try to enforce that Doom cannot Sprint or Run, only walk
  (switch Utilities|FlowControl|Switch|SwitchonEFortMovementStyle NewMovementStyle
    (:Running
      _)
    (:Sprinting
      _))
  (CallFunction|SetCurrentMovementStyle "Walking"))

(event Custom|StartUpdatingHealthWidget
  ;[1 1 1] Health Bar
  (bind _npcstatuswidget (Variables|Default|GetNPCStatusWidget)
    (:then
      (bind _asnpcstatus_widget_aiboss (Utilities|Casting|CastToNPCStatusWidget_AIBoss (UserInterface|GetUserWidgetObject _npcstatuswidget)))
      (Utilities|FlowControl|Sequence
        (:then_0
          (Components|Tick|SetComponentTickEnabled (Variables|Default|GetNPCStatusWidget) (Widget|IsVisible _asnpcstatus_widget_aiboss)))
        (:then_1
          (Default|BindEventtoVisibilityChanged _asnpcstatus_widget_aiboss (delegate Custom|WidgetVisibilityChanged)))))
    (:else)))

(event Custom|WidgetVisibilityChanged
  (params (bVisible :type Boolean))
  ;[1 1 1] Health Bar
  (Components|Tick|SetComponentTickEnabled (Variables|Default|GetNPCStatusWidget) bVisible)
  (if bVisible
    (bind _returnvalue (Utilities|Time|SetTimerbyEvent (EventDispatchers|CreateEvent :!event Custom|UpdateWidgetOffset) 0.1 true))
    (Variables|Default|SetUpdateOffsetTimerHandle _returnvalue)
    (else
      (Utilities|Time|ClearandInvalidateTimerbyHandle (Variables|Default|GetUpdateOffsetTimerHandle)))))

(event Custom|StopUpdatingHealthWidget
  ;[1 1 1] Health Bar
  (bind _npcstatuswidget (Variables|Default|GetNPCStatusWidget)
    (:then
      (bind _asnpcstatus_widget_aiboss (Utilities|Casting|CastToNPCStatusWidget_AIBoss (UserInterface|GetUserWidgetObject _npcstatuswidget)))
      (Default|UnbindEventfromVisibilityChanged _asnpcstatus_widget_aiboss (delegate Custom|WidgetVisibilityChanged))
      (Utilities|Time|ClearandInvalidateTimerbyHandle (Variables|Default|GetUpdateOffsetTimerHandle)))
    (:else)))

(event Custom|UpdateWidgetOffset
  ;[1 1 1] UpdateWidgetOffset
  (bind _returnvalue (AI|GetNPCWidgetOffset (Fortnite|GetFirstLocalFortPlayerController true) self))
  (Variables|/Script/UMG.Widget|SetRenderTransform_Translation_X (Utilities|Struct|MakeWidgetTransform (Math|Vector2D|MakeVector2D 0.0 _returnvalue)) (UserInterface|GetUserWidgetObject (Variables|Default|GetNPCStatusWidget))))

(event Custom|NPCDisableHealthBars
  ;[1 1 1] Debug For Captures
  (Variables|Default|SetNPC%20StatusWidgetHealthBarDisabled true)
  (bind _npcstatuswidget (Variables|Default|GetNPCStatusWidget)
    (:then
      (bind _asnpcstatus_widget_aiboss (Utilities|Casting|CastToNPCStatusWidget_AIBoss (UserInterface|GetUserWidgetObject _npcstatuswidget)))
      (Widget|SetVisibility _asnpcstatus_widget_aiboss "Hidden"))
    (:else)))

(event Custom|NPCEnableHealthBars
  ;[1 1 1] Debug For Captures
  (Variables|Default|SetNPC%20StatusWidgetHealthBarDisabled false))

(fn OnRep_EnableHealthBar
  (if (Variables|Default|GetEnableHealthBar)
    (CallFunction|NPCEnableHealthBars)))

(fn OnRep_HealthInfo
  (CallFunction|UpdatePawnHealthHud)
  (Default|CallOnHealthChange))

(fn UpdatePawnHealthHud
  (bind (_healthpercent_2_7edada384c7908ea98debd9958a249de _shieldpercent_5_b79c63b64aacf275a1621a9315964dcd _maxhealth_6_7d7bef7841b13fb96d2578989cf135bb _maxshield_9_f923e4444e2e476c56fb0e9d9d2f7846) (Utilities|Struct|BreakStructNPCHealthInfo (Variables|Default|GetHealthInfo)))
  ;[1 1 1] On Pawn Health Changed - Update Health Bar
  (Networking|SwitchHasAuthority
    (:Remote
      (bind _array_element (Utilities|Array|ForEachLoopwithBreak (Fortnite|GetLocalFortPlayerControllers true))
        (:LoopBody
          (if (Math|Boolean|ANDBoolean (< (Class|AbilityFunctions|GetDistanceBetweenActors (Pawn|GetControlledPawn _array_element) self) (Variables|Default|GetNPC%20MaxDistanceFromDamageInstigatorToDrawHealthBar)) (not (Variables|Health|GetbIsDying)) (not (Variables|Default|GetNPC%20StatusWidgetHealthBarDisabled)))
            (bind _npcstatuswidget (Variables|Default|GetNPCStatusWidget)
              (:then
                (bind _asnpcstatus_widget_aiboss (Utilities|Casting|CastToNPCStatusWidget_AIBoss (UserInterface|GetUserWidgetObject _npcstatuswidget)))
                (Class|NPCStatusWidgetAIBoss|EventUpdateStatus _asnpcstatus_widget_aiboss _healthpercent_2_7edada384c7908ea98debd9958a249de _shieldpercent_5_b79c63b64aacf275a1621a9315964dcd _maxhealth_6_7d7bef7841b13fb96d2578989cf135bb _maxshield_9_f923e4444e2e476c56fb0e9d9d2f7846 (Variables|Default|GetShowShields) false))
              (:else))))))))

(fn ConstructionScript
  (|CallParentFunction))
